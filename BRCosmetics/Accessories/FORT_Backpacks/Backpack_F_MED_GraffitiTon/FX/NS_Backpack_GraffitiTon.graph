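# Niagara effect for the GraffitiTon backpack cosmetic, adding electrical sparks and three traffic lights.
# ElectricalSparks fires looping random bursts from socket 0 or 1, with gravity, drag and inherited movement.
# TrafficLights pins one partitioned particle per socket 2-4, rendering sprite, mesh and light, with brightness driven by system timelines and time of day.
# User.Variant blends light palettes and sizes; front-end modifiers adjust timing and glow, including camera-angle scaling.
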NiagaraSystem /BRCosmetics/Accessories/FORT_Backpacks/Backpack_F_MED_GraffitiTon/FX/NS_Backpack_GraffitiTon
dependencies {
 usedDataInterfaces [
  &NiagaraDataInterfaceCurve
  &NiagaraDataInterfaceColorCurve
  &NiagaraDataInterfaceArrayFloat2
  &NiagaraDataInterfaceArrayInt32
  &NiagaraDataInterfaceCamera
  &NiagaraDataInterfaceSocketReader
 ]
 usedDynamicInputs [
  &RandomRangeVector
  &FN_FrontEndVector2Modifier
  &SelectVector2DFromArray
  &FNTimeOfDayScalar
  &Boolean_AndOperation
  &SetBoolByFloatComparison
  &Boolean_NotOperation
  &Boolean_OrOperation
 ]
 usedModules [
  &self:SystemSpawnScript.SystemScriptSource.SystemScriptGraph.NiagaraNodeAssignment_3.SetVariables_#1
  &SystemState
  &self:SystemSpawnScript.SystemScriptSource.SystemScriptGraph.NiagaraNodeAssignment_1.SetVariables_#2
  &Timeline
 ]
}
emitters {
 ElectricalSparks {
  emitterUpdateScript [
   /SetVariables(Emitter.RandomSockets:NiagaraInt32=&RandomRangeInt(Minimum=0, Maximum=1)) {isSetParameters=true}
   &EmitterState(
    Life Cycle Mode=::NewEnumerator1
    Loop Delay=&RandomRangeFloat(Minimum=0, Maximum=0.5)
    Loop Duration=&RandomRangeFloat(Minimum=3, Maximum=7, Evaluation Type=::NewEnumerator1)
    Recalculate Duration Each Loop=true
    UseLoopDelay=true
   )
   &SpawnBurst_Instantaneous(
    Spawn Count=&RandomRangeInt(Minimum=3, Maximum=5)
    Use Spawn Probability=false
    UseLoopCountLimit=false
   )
   &SpawnBurst_Instantaneous(Spawn Group=1, Use Spawn Probability=true, UseLoopCountLimit=false) {module=SpawnBurst_Instantaneous001}
  ]
  particleSpawnScript [
   /SetVariables(Particles.VisibilityTag=$Emitter.SpawnGroup) {isSetParameters=true}
   /SetVariables(Particles.PerParticleRandom:NiagaraFloat=&FloatFromCurve(
     FloatCurve=DataInterface() {propertyValues {Curve {keys [
         {arriveTangent=4.1785, interpMode=RCIM_Cubic, tangentMode=RCTM_Break, time=0, value=0}
         {arriveTangent=3.1227, interpMode=RCIM_Cubic, leaveTangent=0.0224, tangentMode=RCTM_Break, time=1, value=1}
        ]}}}
     CurveIndex=&RandomRangeFloat(Minimum=0, Maximum=1)
    )) {isSetParameters=true}
   &InitializeParticle(
    Color Mode=::NewEnumerator0
    Lifetime=&RandomRangeFloat(Minimum=0.3, Maximum=0.5)
    Position Offset Coordinate Space=::Local
    UsePositionOffset=true
    Write Material Random=true
   )
   &SocketLocation(
    Interpolate Socket Sampling=true
    SamplePosOffset=true
    Socket Index=$Emitter.RandomSockets
    Socket Reader=$System.Socket Reader
    Socket Sampling Mode=::NewEnumerator3
    VelocitySpeedLimit=false
   )
   &AddVelocity(
    Additional Quaternion Rotation=false
    Cone Angle=&RandomRangeFloat(Minimum=25, Maximum=90)
    Cone Axis=&ReturnMeshOrientationAxis(Mesh Orientation Axis=::XAxis, Mesh Orientation Quat=$Output.SocketLocation.SampledRotation)
    Module.UseConeFalloff=false
    Module.UseVelDistribution=false
    Override Seed=false
    Rotation Mode=::NewEnumerator3
    Rotation Quaternion:Quat4f=$Particles.SkeletalMeshLocation.SampledRotation
    Velocity Mode=::NewEnumerator2
    Velocity Speed=&MaskFloatBySpawnGroup(Masked Float=&RandomRangeFloat(Minimum=150, Maximum=500), Passthrough Float=0)
   )
   &AddVelocity(
    Rotation Coordinate Space=::World
    Velocity=<0 0 1>
    Velocity Speed Scale=&RandomRangeFloat(Minimum=10, Maximum=40)
   ) {module=AddVelocity001}
   /SetVariables(
    Particles.FlareSpriteSize:Vector2f=<8 8>
    Particles.GlowSpriteSize:Vector2f=&Multiply_Vector2D_ByFloat(Float=0.55, Vector2D=&Lerp_Vector2(A=<5 10>, B=<10 20>, Alpha=$Particles.PerParticleRandom))
    Particles.SparkSpriteSize:Vector2f=&Multiply_Vector2D_ByFloat(Float=0.55, Vector2D=&Lerp_Vector2(A=<1 2>, B=<2 7>, Alpha=$Particles.PerParticleRandom))
   ) {isSetParameters=true}
   &CameraOffset(Camera Offset Amount=&MaskFloatBySpawnGroup(Masked Float=0, Passthrough Float=25))
  ]
  particleUpdateScript [
   &SocketLocation(
    Position Sampling=::NewEnumerator1
    SamplePosOffset=false
    Socket Index=$Emitter.RandomSockets
    Socket Reader=$System.Socket Reader
    Socket Sampling Mode=::NewEnumerator3
    VelocitySpeedLimit=false
   ) {module=SocketLocation001}
   /SetVariables(Particles.Position_LightGlow:NiagaraPosition=$Output.SocketLocation001.SampledPosition) {isSetParameters=true}
   /SetVariables(Particles.SparkSpriteColor:LinearColor=&MultiplyLinearColor(
     Linear Color=&MultiplyLinearColor(
      Linear Color=&ColorFromCurve(ColorCurve=DataInterface() {propertyValues {
         AlphaCurve {keys [<at 0.01 1>,<at 0.9198 1 RCIM_Cubic>,<at 1 0>]}
         BlueCurve {keys [
           <at 0.0173 33>
           {arriveTangent=-211.0984, interpMode=RCIM_Cubic, leaveTangent=-211.0984, time=0.0588, value=27.192}
           <at 0.0857 1.1>
           <at 0.1736 0>
           <at 0.4906 0>
           <at 0.7101 0>
           <at 0.9394 0>
          ]}
         GreenCurve {keys [
           <at 0.0173 9.9>
           {arriveTangent=-49.2562, interpMode=RCIM_Cubic, leaveTangent=-49.2562, time=0.0588, value=33}
           <at 0.0857 8.36>
           <at 0.1736 2.2>
           <at 0.4906 1.0705>
           <at 0.7101 0.1833>
           <at 0.9394 0.0762>
          ]}
         RedCurve {keys [
           <at 0.0173 0>
           {arriveTangent=19.1908, interpMode=RCIM_Cubic, leaveTangent=19.1908, time=0.0588, value=1.32}
           <at 0.0857 11>
           <at 0.1736 3>
           <at 0.4906 2>
           <at 0.7101 1>
           <at 0.9394 1>
          ]}
        }})
      Scale Factor=&MakeFloat4FromVectorAndFloat("Vector (RGB)"=&VectorFromFloat(Value=0.8), "Float (Alpha)"=1)
     )
     Scale Factor=&MakeFloat4FromVectorAndFloat(
      "Vector (RGB)"=&VectorFromFloat(Value=&FN_FrontEndFloatModifier(Gameplay=1, Front End=0.4))
      "Float (Alpha)"=1
     )
    )) {isSetParameters=true}
   /SetVariables(Particles.GlowSpriteColor:LinearColor=&MultiplyLinearColor(
     Linear Color=$Particles.SparkSpriteColor
     Scale Factor=&MakeFloat4FromVectorAndFloat("Vector (RGB)"=<1 1 1>, "Float (Alpha)"=0.03)
    )) {isSetParameters=true}
   /SetVariables(Particles.FlareSpriteColor:LinearColor=&MultiplyLinearColor(
     Linear Color=&MultiplyLinearColor(
      Linear Color=&ColorFromCurve(ColorCurve=DataInterface() {propertyValues {
         AlphaCurve {keys [<at 0.01 1>,<at 0.9198 1 RCIM_Cubic>,<at 1 0>]}
         BlueCurve {keys [
           <at 0.0166 33>
           {arriveTangent=-425.625, interpMode=RCIM_Cubic, leaveTangent=-425.625, time=0.0522, value=27.192}
           <at 0.0915 1.1>
           <at 0.1257 0>
           <at 0.4575 0>
           <at 0.7101 0>
           <at 0.9394 0>
          ]}
         GreenCurve {keys [
           <at 0.0166 9.9>
           {arriveTangent=-20.5472, interpMode=RCIM_Cubic, leaveTangent=-20.5472, time=0.0522, value=33}
           <at 0.0915 8.36>
           <at 0.1257 2.2>
           <at 0.4575 1.0705>
           <at 0.7101 0.1833>
           <at 0.9394 0.0762>
          ]}
         RedCurve {keys [
           <at 0.0166 0>
           {arriveTangent=146.7672, interpMode=RCIM_Cubic, leaveTangent=146.7672, time=0.0522, value=1.32}
           <at 0.0915 11>
           <at 0.1257 3>
           <at 0.4575 2>
           <at 0.7101 1>
           <at 0.9394 1>
          ]}
        }})
      Scale Factor=&MakeFloat4FromVectorAndFloat("Vector (RGB)"=&VectorFromFloat(Value=0.8), "Float (Alpha)"=1)
     )
     Scale Factor=&MakeFloat4FromVectorAndFloat(
      "Vector (RGB)"=&VectorFromFloat(Value=&FN_FrontEndFloatModifier(Gameplay=0.75, Front End=0.75))
      "Float (Alpha)"=&FloatFromCurve(
       FloatCurve=DataInterface() {propertyValues {Curve {keys [
           {tangentMode=RCTM_User, time=0, value=1}
           <at 0.0756 -0.7485>
           <at 0.1284 2.2279>
           <at 0.3098 -0.7559>
           <at 0.4525 0.9206>
           <at 0.525 0.0265>
           <at 0.6661 0.3676>
           <at 0.7436 1.6548>
           <at 0.8248 0.1441>
           <at 0.9286 0.8147>
           {tangentMode=RCTM_User, time=1, value=0}
          ]}}}
       CurveIndex=&NormalizeFloat(
        Value=&Multiply_Float(A=$Particles.NormalizedAge, B=&RandomRangeFloat(Minimum=2.5, Maximum=4))
        Min=0
        Max=1
       )
       Scale Curve=0.1
      )
     )
    )) {isSetParameters=true}
   /SetVariables(Particles.LightColor=&MultiplyLinearColor(
     Linear Color=$Particles.FlareSpriteColor
     Scale Factor=&MakeFloat4FromVectorAndFloat("Vector (RGB)"=&VectorFromFloat(Value=3), "Float (Alpha)"=1)
    )) {isSetParameters=true}
   &ParticleState()
   &Drag(
    Drag=&FloatFromCurve(
     FloatCurve=DataInterface() {propertyValues {Curve {keys [{interpMode=RCIM_Cubic, tangentMode=RCTM_User, time=0, value=1}, {interpMode=RCIM_Cubic, tangentMode=RCTM_User, time=1, value=0}]}}}
     Scale Curve=&Lerp_Float(A=17, B=5, Alpha=$Particles.PerParticleRandom)
    )
    Use Linear Drag=true
    Use Rotational Drag=true
   )
   &GravityForce(Gravity=&MaskVectorBySpawnGroup(Masked Vector=<0 0 -980>, Passthrough Vector=<0 0 0>))
   &InheritSourceMovement(
    Applied Position Scale=&VectorFromFloat(Value=&FloatFromCurve(FloatCurve=DataInterface() {propertyValues {Curve {
         defaultValue=1
         keys [
          {interpMode=RCIM_Cubic, tangentMode=RCTM_User, time=0, value=1}
          {arriveTangent=-1.9054, interpMode=RCIM_Cubic, leaveTangent=-1.9054, tangentMode=RCTM_User, time=1, value=0}
         ]
        }}}))
    ApplyForce=false
    ApplyPosition=true
    ApplyPositionFalloff=false
   ) {module=InheritSourceMovement002}
   &SolveForcesAndVelocity(Clamp Velocity=false, Limit Acceleration=false)
  ]
  rendererClasses [&NiagaraSpriteRendererProperties, &NiagaraLightRendererProperties]
  renderers [
   {rendererClass=&NiagaraSpriteRendererProperties}
   {rendererClass=&NiagaraSpriteRendererProperties, rendererIndex=1}
   {rendererClass=&NiagaraLightRendererProperties, rendererIndex=2}
   {rendererClass=&NiagaraSpriteRendererProperties, rendererIndex=3}
  ]
 }
 TrafficLights {
  emitterUpdateScript [
   &EmitterState(
    Inactive Response=::NewEnumerator1
    Life Cycle Mode=::NewEnumerator1
    Loop Behavior=::NewEnumerator1
    Loop Duration Mode=::NewEnumerator1
    UseLoopDelay=false
   )
   &SpawnBurst_Instantaneous(Spawn Count=3, Use Spawn Probability=true, UseLoopCountLimit=false)
  ]
  particleSpawnScript [
   &PartitionParticles(Number Of Partitions=3)
   &InitializeParticle(
    Color=&LerpLinearColors(
     StartColor=&MergeHelper_ColorSwitch(
      0=<rgba 5 0.951 0.8824 0.05>
      1=<rgba 5 3.4933 0.8824 0.05>
      2=<rgba 0.8824 5 4.5882 0.05>
      3=?
      Index=$Particles.PartitionParticles.Partition
      Num=::NewEnumerator1
     )
     EndColor=&MergeHelper_ColorSwitch(
      0=<rgba 5 0 4.2795 0.05>
      1=<rgba 0.616 0.519 12 0.05>
      2=<rgba 0.4527 5.5692 6 0.025>
      3=?
      Index=$Particles.PartitionParticles.Partition
      Num=::NewEnumerator1
     )
     LerpFactor=$User.Variant
    )
    Mesh Scale=&FN_FrontEndVector3Modifier(Gameplay=<0.15 0.24 0.24>, Front End=<0.15 0.24 0.24>)
    Mesh Scale Mode=::NewEnumerator4
    Sprite Size Mode=::NewEnumerator3
    Uniform Sprite Size=&FN_FrontEndFloatModifier(
     Gameplay=14
     Front End=&Lerp_Float(
      A=13
      B=&MergeHelper_FloatSwitch(0=12, 1=12, 2=11, 3=?, Index=$Particles.PartitionParticles.Partition, Num=::NewEnumerator1)
      Alpha=$User.Variant
     )
    )
    UsePositionOffset=false
    Write Material Random=true
   )
   &InitialMeshOrientation(Mesh Orientation Mode=::NewEnumerator7, Rotation=<0 0 -0.25>, Use Rotation Vector=true)
  ]
  particleUpdateScript [
   &SolveForcesAndVelocity(Clamp Velocity=false, Limit Acceleration=false)
   &ParticleState(Kill Particles When Lifetime Has Elapsed=false, Let Infinitely Lived Particles Die When Emitter Deactivates=true)
   &SocketLocation(
    Interpolate Socket Sampling=true
    SamplePosOffset=true
    Socket Index=&SelectIntFromArray(Direct Array Index=$Particles.PartitionParticles.Partition, Int Selection Array=DataInterface() {propertyValues {IntData [2,3,4]}})
    Socket Reader=$System.Socket Reader
    Socket Sampling Mode=::NewEnumerator3
    VelocitySpeedLimit=false
   )
   /SetVariables(Particles.GlowMeshpos:NiagaraPosition=&AddVectorToPosition(
     Position=$Particles.Position
     Vector=&RotateVectorByQuaternion(Vector To Rotate=<0 2.9 0>, Quaternion=$Output.SocketLocation.SampledRotation)
    )) {isSetParameters=true}
   &CameraOffset(Camera Offset Amount=&FN_FrontEndFloatModifier(Gameplay=0, Front End=0))
   &ScaleColor(
    Scale Alpha=&FN_FrontEndFloatModifier(
     Gameplay=&Multiply_Float(
      A=&MergeHelper_FloatSwitch(
       0=&FloatFromCurve(
        FloatCurve=DataInterface() {propertyValues {Curve {
           defaultValue=0
           keys [{interpMode=RCIM_Cubic, tangentMode=RCTM_User, time=0, value=0}, {interpMode=RCIM_Cubic, tangentMode=RCTM_User, time=1, value=1}]
          }}}
        CurveIndex=$System.Timeline.NormalizedTime
        Scale Curve=5
       )
       1=&FloatFromCurve(
        FloatCurve=DataInterface() {propertyValues {Curve {
           defaultValue=0
           keys [{interpMode=RCIM_Cubic, tangentMode=RCTM_User, time=0, value=0}, {interpMode=RCIM_Cubic, tangentMode=RCTM_User, time=1, value=1}]
          }}}
        CurveIndex=$System.Timeline002.NormalizedTime
        Scale Curve=&Lerp_Float(A=6, B=8, Alpha=$User.Variant)
       )
       2=&FloatFromCurve(
        FloatCurve=DataInterface() {propertyValues {Curve {
           defaultValue=0
           keys [{interpMode=RCIM_Cubic, tangentMode=RCTM_User, time=0, value=0}, {interpMode=RCIM_Cubic, tangentMode=RCTM_User, time=1, value=1}]
          }}}
        CurveIndex=$System.Timeline001.NormalizedTime
        Scale Curve=4
       )
       3=?
       Index=$Particles.PartitionParticles.Partition
       Num=::NewEnumerator1
      )
      B=$System.TimeOfDayBrightnessMultiplier
     )
     Front End=&Multiply_Float(
      A=&MergeHelper_FloatSwitch(
       0=&FloatFromCurve(
        FloatCurve=DataInterface() {propertyValues {Curve {
           defaultValue=0
           keys [{interpMode=RCIM_Cubic, tangentMode=RCTM_User, time=0, value=0}, {interpMode=RCIM_Cubic, tangentMode=RCTM_User, time=1, value=1}]
          }}}
        CurveIndex=$System.Timeline003.NormalizedTime
        Scale Curve=4
       )
       1=&FloatFromCurve(
        FloatCurve=DataInterface() {propertyValues {Curve {
           defaultValue=0
           keys [{interpMode=RCIM_Cubic, tangentMode=RCTM_User, time=0, value=0}, {interpMode=RCIM_Cubic, tangentMode=RCTM_User, time=1, value=1}]
          }}}
        CurveIndex=$System.Timeline004.NormalizedTime
        Scale Curve=3
       )
       2=&FloatFromCurve(
        FloatCurve=DataInterface() {propertyValues {Curve {
           defaultValue=0
           keys [{interpMode=RCIM_Cubic, tangentMode=RCTM_User, time=0, value=0}, {interpMode=RCIM_Cubic, tangentMode=RCTM_User, time=1, value=1}]
          }}}
        CurveIndex=$System.Timeline005.NormalizedTime
        Scale Curve=3
       )
       3=?
       Index=$Particles.PartitionParticles.Partition
       Num=::NewEnumerator1
      )
      B=$System.TimeOfDayBrightnessMultiplier
     )
    )
    ScaleA=true
    ScaleRGB=false
   ) {module=ScaleColor001}
   &FN_Light_Attributes(
    Color=$Particles.Color
    Create Color Attribute=true
    Create Exponent Attribute=true
    Create Position Attribute=true
    Create Radius Attribute=true
    Exponent=3
    Position=&AddVectorToPosition(Position=$Particles.Position, Vector=<0 0 0>)
    Radius=&Multiply_Float(
     A=&FN_FrontEndFloatModifier(
      Gameplay=&MergeHelper_FloatSwitch(
       0=&Lerp_Float(A=65, B=55, Alpha=$User.Variant)
       1=&Lerp_Float(A=65, B=50, Alpha=$User.Variant)
       2=&Lerp_Float(A=65, B=50, Alpha=$User.Variant)
       3=?
       Index=$Particles.PartitionParticles.Partition
       Num=::NewEnumerator1
      )
      Front End=&MergeHelper_FloatSwitch(
       0=50
       1=50
       2=&Lerp_Float(A=48, B=47, Alpha=$User.Variant)
       3=?
       Index=$Particles.PartitionParticles.Partition
       Num=::NewEnumerator1
      )
     )
     B=&FN_FrontEndFloatModifier(
      Gameplay=&Multiply_Float(
       A=&MergeHelper_FloatSwitch(
        0=&FloatFromCurve(
         FloatCurve=DataInterface() {propertyValues {Curve {
            defaultValue=0
            keys [{interpMode=RCIM_Cubic, tangentMode=RCTM_User, time=0, value=0}, {interpMode=RCIM_Cubic, tangentMode=RCTM_User, time=1, value=1}]
           }}}
         CurveIndex=$System.Timeline.NormalizedTime
        )
        1=&FloatFromCurve(
         FloatCurve=DataInterface() {propertyValues {Curve {
            defaultValue=0
            keys [{interpMode=RCIM_Cubic, tangentMode=RCTM_User, time=0, value=0}, {interpMode=RCIM_Cubic, tangentMode=RCTM_User, time=1, value=1}]
           }}}
         CurveIndex=$System.Timeline002.NormalizedTime
        )
        2=&FloatFromCurve(
         FloatCurve=DataInterface() {propertyValues {Curve {
            defaultValue=0
            keys [{interpMode=RCIM_Cubic, tangentMode=RCTM_User, time=0, value=0}, {interpMode=RCIM_Cubic, tangentMode=RCTM_User, time=1, value=1}]
           }}}
         CurveIndex=$System.Timeline001.NormalizedTime
        )
        3=?
        Index=$Particles.PartitionParticles.Partition
        Num=::NewEnumerator1
       )
       B=$System.TimeOfDayBrightnessMultiplier
      )
      Front End=&Multiply_Float(
       A=&MergeHelper_FloatSwitch(
        0=&FloatFromCurve(
         FloatCurve=DataInterface() {propertyValues {Curve {
            defaultValue=0
            keys [{interpMode=RCIM_Cubic, tangentMode=RCTM_User, time=0, value=0}, {interpMode=RCIM_Cubic, tangentMode=RCTM_User, time=1, value=1}]
           }}}
         CurveIndex=$System.Timeline003.NormalizedTime
        )
        1=&FloatFromCurve(
         FloatCurve=DataInterface() {propertyValues {Curve {
            defaultValue=0
            keys [{interpMode=RCIM_Cubic, tangentMode=RCTM_User, time=0, value=0}, {interpMode=RCIM_Cubic, tangentMode=RCTM_User, time=1, value=1}]
           }}}
         CurveIndex=$System.Timeline004.NormalizedTime
        )
        2=&FloatFromCurve(
         FloatCurve=DataInterface() {propertyValues {Curve {
            defaultValue=0
            keys [{interpMode=RCIM_Cubic, tangentMode=RCTM_User, time=0, value=0}, {interpMode=RCIM_Cubic, tangentMode=RCTM_User, time=1, value=1}]
           }}}
         CurveIndex=$System.Timeline005.NormalizedTime
        )
        3=?
        Index=$Particles.PartitionParticles.Partition
        Num=::NewEnumerator1
       )
       B=$System.TimeOfDayBrightnessMultiplier
      )
     )
    )
    Scale Color=&Lerp_Float(A=30, B=60, Alpha=$User.Variant)
   )
   /SetVariables(
    Particles.GlowColor=&Multiply_LinearColorByFloat(
     Linear Color=$Particles.Color
     Float=&FN_FrontEndFloatModifier(
      Gameplay=1
      Front End=&Lerp_Float(
       A=0.5
       B=1
       Alpha=&NormalizeFloat(
        Value=&DotProduct(Vector A=&ReturnCameraVector(Camera Vectors=::NewEnumerator1, Coordinate Space=::Simulation), Vector B=<0 1 0>)
        Min=0
        Max=1
       )
      )
     )
    )
    Particles.GlowSize=&Multiply_Vector2D_ByFloat(
     Float=&FN_FrontEndFloatModifier(
      Gameplay=1
      Front End=&Lerp_Float(
       A=0.5
       B=0.8
       Alpha=&NormalizeFloat(
        Value=&DotProduct(Vector A=&ReturnCameraVector(Camera Vectors=::NewEnumerator1, Coordinate Space=::Simulation), Vector B=<0 1 0>)
        Min=0
        Max=1
       )
      )
     )
     Vector2D=$Particles.SpriteSize
    )
    Particles.SpriteCameraOffset:NiagaraFloat=&FN_FrontEndFloatModifier(
     Gameplay=1
     Front End=&Lerp_Float(
      A=2
      B=3
      Alpha=&NormalizeFloat(
       Value=&DotProduct(Vector A=&ReturnCameraVector(Camera Vectors=::NewEnumerator1, Coordinate Space=::Simulation), Vector B=<0 1 0>)
       Min=0
       Max=1
      )
     )
    )
   ) {isSetParameters=true}
  ]
  rendererClasses [
   &NiagaraSpriteRendererProperties
   &NiagaraLightRendererProperties
   &NiagaraMeshRendererProperties
  ]
  renderers [
   {rendererClass=&NiagaraSpriteRendererProperties}
   {rendererClass=&NiagaraLightRendererProperties, rendererIndex=1}
   {rendererClass=&NiagaraMeshRendererProperties, rendererIndex=2}
  ]
 }
}
summary {
 systemName=NS_Backpack_GraffitiTon
 emitters [
  {
   emitterName=ElectricalSparks
   rendererClasses [&NiagaraSpriteRendererProperties, &NiagaraLightRendererProperties]
  }
  {
   emitterName=TrafficLights
   rendererClasses [
    &NiagaraSpriteRendererProperties
    &NiagaraLightRendererProperties
    &NiagaraMeshRendererProperties
   ]
  }
 ]
 userVariables [
  User.bIsFrontEnd {type=NiagaraBool}
  User.bIsFrontEndPreview {type=NiagaraBool}
  User.IsCrouched {type=NiagaraBool}
  User.IsSkydiving {type=NiagaraBool}
  User.PlayerVelocityNormalized {type=NiagaraFloat, defaultValue=0, description=""}
  User.Variant {type=NiagaraFloat, defaultValue=1, description=""}
 ]
}
